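# Niagara system for the idle effects of the WaveLaces cosmetic shoes, showing water effects at both feet while User.isStage3 is set.
# A one-shot burst spawns splash, droplet and chunk sprites, while WaterDrops keeps emitting drops from skeletal-mesh sockets.
# Anim-trail ribbons appear only when running, not skydiving and with foot FX visible.
# Particle partitions and spawn groups split left and right feet for placement and shoe scale, and time of day drives tinting.
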NiagaraSystem /CosmeticShoes/Assets/FORT_Shoes/Shoe_WaveLaces/FX/NS_Shoe_WaveLaces_Idle
dependencies {
 usedDataInterfaces [
  &NiagaraDataInterfaceSpriteRendererInfo
  &NiagaraDataInterfaceArrayInt32
  &NiagaraDataInterfaceCurve
  &NiagaraDataInterfaceColorCurve
  &NiagaraDataInterfacePlatformSet
  &NiagaraDataInterfaceSkeletalMesh
 ]
 usedDynamicInputs [
  &SimulationPosition
  &MaskVector2dBySpawnGroup
  &Vector2DFromFloat
  &RandomRangeVector2D
  &ClampFloat
  &Boolean_AndOperation
  /Game/Effects/Niagara/DynamicInputs/Logic/FN_IsRunningCheck
  &SetBoolByFloatComparison
  &VectorLength
  &Divide_Float
  &FNTimeOfDayColor
 ]
 usedModules [
  &self:SystemSpawnScript.SystemScriptSource.SystemScriptGraph.NiagaraNodeAssignment_4.SetVariables_#1
  &self:SystemSpawnScript.SystemScriptSource.SystemScriptGraph.NiagaraNodeAssignment_3.SetVariables_#2
  &SystemState
  &self:SystemSpawnScript.SystemScriptSource.SystemScriptGraph.NiagaraNodeAssignment_1.SetVariables_#3
  &self:SystemSpawnScript.SystemScriptSource.SystemScriptGraph.NiagaraNodeAssignment_2.SetVariables_#4
  &Timeline
  &self:SystemSpawnScript.SystemScriptSource.SystemScriptGraph.NiagaraNodeAssignment_0.SetVariables_#5
 ]
}
emitters {
 DropsFishes {
  emitterUpdateScript [
   &EmitterState()
   &DoOnce(Trigger Condition:NiagaraBool=$User.isStage3)
   &self:TriggerSpawn(
    Spawn Count:NiagaraInt32=&RandomRangeInt(Minimum=9, Maximum=12)
    Trigger:NiagaraBool=$Emitter.DoOnce.Execute
   ) {module=TriggerSpawn}
  ]
  particleSpawnScript [
   &PartitionParticles()
   &InitializeParticle(
    Color=<rgba 1 1 1 0.9>
    Lifetime=&UniformRangedFloat(Minimum=0.4, Maximum=0.6)
    Mass Mode=::NewEnumerator1
    Mesh Scale Mode=::NewEnumerator3
    Mesh Uniform Scale=1.25
    Sprite Size=&Multiply_Vector2D_ByFloat(
     Float=&MaskFloatBySpawnGroup(
      Masked Float=&Multiply_Float(A=1, B=&MakeFloatFromVector(Vector=$User.ShoeScale_L, Channel=::NewEnumerator0))
      Passthrough Float=&Multiply_Float(A=1, B=&MakeFloatFromVector(Vector=$User.ShoeScale_R, Channel=::NewEnumerator0))
      Spawn Group Mask=$Particles.PartitionParticles.Partition
     )
     Vector2D=&Multiply_Vector2D_ByFloat(
      Float=&RandomRangeFloat(Minimum=1, Maximum=2)
      Vector2D=&UniformRangedVector2D(Minimum=<1 2>, Maximum=<2 4>)
     )
    )
    Sprite Size Mode=::NewEnumerator4
    UsePositionOffset=false
    Write Material Random=true
   ) {module=InitializeParticle001}
   /Niagara/Modules/Spawn/Location/SphereLocation(MaskSpawnGroup=false, Module.Override Randomness=false, Sphere Radius=5, Surface Only=false)
   &AddVelocity(
    Cone Angle=&MergeHelper_FloatSwitch(0=90, 1=70, 2=?, 3=?, Index=&SpawnGroup(), Num=::NewEnumerator0)
    Cone Axis=<0 0 1>
    Module.UseConeFalloff=false
    Module.UseVelDistribution=false
    Override Seed=false
    Rotation Coordinate Space=::World
    Velocity Mode=::NewEnumerator2
    Velocity Speed=&MergeHelper_FloatSwitch(
     0=&RandomRangeFloat(Minimum=100, Maximum=250)
     1=&RandomRangeFloat(Minimum=200, Maximum=350)
     2=?
     3=?
     Index=&SpawnGroup()
     Num=::NewEnumerator0
    )
   )
   /Niagara/Modules/Update/SubUV/SubUVAnimation(SubUV Animation Mode=::NewEnumerator1) {module=SubUVAnimation001}
  ]
  particleUpdateScript [
   &ParticleState()
   &SkeletalMeshLocation(
    Bone / Socket Index=&SelectIntFromArray(Direct Array Index=$Particles.PartitionParticles.Partition, Int Selection Array=DataInterface() {propertyValues {IntData [1,3]}})
    Bone Sampling Mode=::NewEnumerator4
    MaskSpawnGroup=false
    Orientation Sampling=::NewEnumerator1
    Sampled Position Offset=&Multiply_Vector(
     A=&AddVectorToPosition(
      Position=&RandomRangeVector(Minimum=<-5 0 -6>, Maximum=<5 0 6>)
      Vector=&MergeHelper_VectorSwitch(0=<-4 0 4>, 1=<-4 0 0>, 2=?, 3=?, Index=$Particles.PartitionParticles.Partition, Num=::NewEnumerator0)
     )
     B=$User.ShoeScale_L
    )
    Skeletal Mesh=$System.Skeletal Mesh Reader
   ) {module=SkeletalMeshLocation001}
   &Drag(Drag=&RandomRangeFloat(Minimum=1, Maximum=2), Use Linear Drag=true, Use Rotational Drag=true)
   &GravityForce(Gravity=&MakeVector(X=0, Y=0, Z=&RandomRangeFloat(Minimum=-700, Maximum=-980))) {module=GravityForce001}
   &SolveForcesAndVelocity(Clamp Velocity=false, Limit Acceleration=false)
   &ScaleSpriteSize(Scale Factor=&Multiply_Vector2D_ByFloat(
     Float=&FloatFromCurve(FloatCurve=DataInterface() {propertyValues {Curve {keys [<at 0 0.5>,<at 1 1>]}}})
     Vector2D=&Multiply_Vector2D_ByFloat(Float=$System.ScaleShoes, Vector2D=<1 1.5>)
    ))
   &ScaleColor(
    Scale Alpha=&FloatFromCurve(FloatCurve=DataInterface() {propertyValues {Curve {keys [<at 0 1>,<at 0.75 1>,<at 1 0>]}}})
    Scale RGB=&Multiply_VectorByFloat(
     Vector=&MakeVectorFromLinearColorRGB(Color=$System.TODColor)
     Float=&MakeFloatFromVector(
      Vector=&VectorFromFloat(Value=&FloatFromCurve(FloatCurve=DataInterface() {propertyValues {Curve {keys [<at 0 2 RCIM_Cubic>,<at 0.25 1 RCIM_Cubic>]}}}))
      Channel=::NewEnumerator0
     )
    )
    ScaleA=false
    ScaleRGB=true
   )
   &DynamicMaterialParameters(
    Index 0 Param 1=&RandomRangeFloat(Minimum=0, Maximum=1)
    Index 0 Param 2=&FloatFromCurve(FloatCurve=DataInterface() {propertyValues {Curve {keys [<at 0.35 0.45>,<at 1 0.9>]}}})
    Index 0 Param 3=&FloatFromCurve(FloatCurve=DataInterface() {propertyValues {Curve {keys [{arriveTangent=2.6651, interpMode=RCIM_Cubic, leaveTangent=2.6651, tangentMode=RCTM_User, time=0, value=0}, <at 1 1 RCIM_Cubic>]}}})
    Index 0 Param 4=&FloatFromCurve(FloatCurve=DataInterface() {propertyValues {Curve {keys [<at 0 0 RCIM_Cubic>, {arriveTangent=2.332, interpMode=RCIM_Cubic, leaveTangent=2.332, tangentMode=RCTM_User, time=1, value=1}]}}})
    Module.Param0WriteEnabled=true
    Module.Param1WriteEnabled=true
    Module.Param2WriteEnabled=false
    Module.Param3WriteEnabled=true
   ) {module=DynamicMaterialParameters001}
  ]
  rendererClasses [&NiagaraSpriteRendererProperties]
  renderers [{rendererClass=&NiagaraSpriteRendererProperties}]
 }
 GravityChunks {
  emitterUpdateScript [
   &EmitterState()
   &DoOnce(Trigger Condition:NiagaraBool=$User.isStage3)
   &self:TriggerSpawn(Spawn Count=24, Trigger:NiagaraBool=$Emitter.DoOnce.Execute) {module=TriggerSpawn}
  ]
  particleSpawnScript [
   &PartitionParticles()
   &InitializeParticle(
    Color=<rgba 1 1 1 0.8>
    Lifetime Max=0.7
    Lifetime Min=0.4
    Lifetime Mode=::NewEnumerator1
    Sprite Rotation Mode=::NewEnumerator2
    Sprite Size Mode=::NewEnumerator3
    Sprite UV Mode=::NewEnumerator3
    Uniform Sprite Size=&Multiply_Float(
     A=&RandomRangeFloat(Minimum=2, Maximum=8)
     B=&MaskFloatBySpawnGroup(
      Masked Float=&Multiply_Float(A=1, B=&MakeFloatFromVector(Vector=$User.ShoeScale_L, Channel=::NewEnumerator0))
      Passthrough Float=&Multiply_Float(A=1, B=&MakeFloatFromVector(Vector=$User.ShoeScale_R, Channel=::NewEnumerator0))
      Spawn Group Mask=$Particles.PartitionParticles.Partition
     )
    )
    UsePositionOffset=false
    Write Material Random=true
   )
   &SkeletalMeshLocation(
    Bone / Socket Index=&SelectIntFromArray(Direct Array Index=$Particles.PartitionParticles.Partition, Int Selection Array=DataInterface() {propertyValues {IntData [1,3]}})
    Bone Sampling Mode=::NewEnumerator4
    MaskSpawnGroup=false
    Sampled Position Offset=&Multiply_Vector(
     A=&AddVectorToPosition(
      Position=&RandomRangeVector(Minimum=<-5 0 -6>, Maximum=<5 0 6>)
      Vector=&MergeHelper_VectorSwitch(0=<-4 0 4>, 1=<-4 0 0>, 2=?, 3=?, Index=$Particles.PartitionParticles.Partition, Num=::NewEnumerator0)
     )
     B=$User.ShoeScale_L
    )
    Skeletal Mesh=$System.Skeletal Mesh Reader
   ) {module=SkeletalMeshLocation001}
   /Niagara/Modules/Update/SubUV/V2/SubUVAnimation(End Frame=3, SubUV Animation Mode:ENiagara_SubUVLookupMode=::NewEnumerator2)
   &ShapeLocation(
    MaskSpawnGroup=false
    Override Seed=false
    Sphere Radius=&Multiply_Float(A=&RandomRangeFloat(Minimum=10, Maximum=15), B=1)
    UseOwnerScale=true
   )
   &CameraOffset(Camera Offset Amount=&Multiply_Float(A=15, B=$System.GlobalScale))
   &AddVelocity(
    Cone Angle=120
    Cone Axis=<0 0 1>
    Module.UseConeFalloff=false
    Module.UseVelDistribution=false
    Override Seed=false
    Velocity Mode=::NewEnumerator2
    Velocity Speed=&Multiply_Float(A=&RandomRangeFloat(Minimum=100, Maximum=250), B=1)
   )
   /SetVariables(Particles.Random=&UniformAOrBFloat(A=-1)) {isSetParameters=true}
  ]
  particleUpdateScript [
   &ParticleState()
   &SpriteRotationRate(Rotation Rate=&Multiply_Float(
     A=&MakeFloatFromInt(Int=&SelectIntFromArray(
       Array Sampling Mode=::NewEnumerator1
       Direct Array Index=?
       Int Selection Array=DataInterface() {propertyValues {IntData [-1,1]}}
       Random Seed=0
      ))
     B=&FloatFromCurve(
      FloatCurve=DataInterface() {propertyValues {Curve {keys [{arriveTangent=-2, interpMode=RCIM_Cubic, time=0.15, value=1}, {interpMode=RCIM_Cubic, leaveTangent=-8.0089, time=0.8, value=0}]}}}
      Scale Curve=&RandomRangeFloat(Minimum=100, Maximum=300)
     )
    ))
   &DynamicMaterialParameters(
    Index 0 Param 2=&FloatFromCurve(FloatCurve=DataInterface() {propertyValues {Curve {keys [{interpMode=RCIM_Cubic, tangentMode=RCTM_Break, time=0.05, value=0.3}, {interpMode=RCIM_Cubic, tangentMode=RCTM_Break, time=0.95, value=1}]}}})
    Index 0 Param 3=&FloatFromCurve(FloatCurve=DataInterface() {propertyValues {Curve {keys [
         {interpMode=RCIM_Cubic, tangentMode=RCTM_Break, time=0.05, value=-0.5}
         {interpMode=RCIM_Cubic, tangentMode=RCTM_Break, time=0.95, value=-1.43}
        ]}}})
    Index 0 Param 4=&FloatFromCurve(FloatCurve=DataInterface() {propertyValues {Curve {keys [{arriveTangent=2.6651, interpMode=RCIM_Cubic, leaveTangent=2.6651, tangentMode=RCTM_User, time=0, value=0.1}, <at 0.6 1.333 RCIM_Cubic>]}}})
    Module.Param0WriteEnabled=true
    Module.Param1WriteEnabled=true
    Module.Param2WriteEnabled=true
    Module.Param3WriteEnabled=true
   )
   &GravityForce(Gravity=&MakeVector(X=0, Y=0, Z=&RandomRangeFloat(Minimum=-800, Maximum=-1000)))
   &Drag(Drag=5, Use Linear Drag=true, Use Rotational Drag=true)
   &SolveForcesAndVelocity(Clamp Velocity=false, Limit Acceleration=false)
   &ScaleColor(
    Scale Alpha=&FloatFromCurve(FloatCurve=DataInterface() {propertyValues {Curve {keys [
         {interpMode=RCIM_Cubic, tangentMode=RCTM_Break, time=0, value=0}
         {interpMode=RCIM_Cubic, tangentMode=RCTM_Break, time=0.15, value=1}
         {interpMode=RCIM_Cubic, tangentMode=RCTM_User, time=0.5, value=1}
         {interpMode=RCIM_Cubic, tangentMode=RCTM_User, time=1, value=0}
        ]}}})
    Scale RGB=&Multiply_VectorByFloat(
     Vector=&MakeVectorFromLinearColorRGB(Color=$System.TODColor)
     Float=&MakeFloatFromVector(
      Vector=&VectorFromFloat(Value=&FloatFromCurve(FloatCurve=DataInterface() {propertyValues {Curve {keys [<at 0 2 RCIM_Cubic>,<at 0.25 1 RCIM_Cubic>]}}}))
      Channel=::NewEnumerator0
     )
    )
    ScaleA=false
    ScaleRGB=true
   )
   &ScaleSpriteSize(
    Uniform Curve Scale=$System.ScaleShoes
    Uniform Curve Sprite Scale=DataInterface() {propertyValues {Curve {
       defaultValue=0
       keys [
        {arriveTangent=2.5867, interpMode=RCIM_Cubic, leaveTangent=2.5867, tangentMode=RCTM_User, time=0, value=0.5}
        {interpMode=RCIM_Cubic, tangentMode=RCTM_User, time=0.5, value=1}
        {tangentMode=RCTM_User, time=1, value=0}
       ]
      }}}
   )
  ]
  rendererClasses [&NiagaraSpriteRendererProperties]
  renderers [{rendererClass=&NiagaraSpriteRendererProperties}]
 }
 SingleSplash {
  emitterUpdateScript [
   &EmitterState()
   &DoOnce(Trigger Condition:NiagaraBool=$User.isStage3)
   &self:TriggerSpawn(Spawn Count=2, Trigger:NiagaraBool=$Emitter.DoOnce.Execute) {module=TriggerSpawn}
  ]
  particleSpawnScript [
   &PartitionParticles()
   &InitializeParticle(
    Color=&RandomRangeLinearColor(Minimum=<rgba 0.11 0.6875 1.1 1>, Maximum=<rgba 0.1 0.55 1 1>)
    Lifetime=0.5
    Sprite Rotation Mode=::NewEnumerator2
    Sprite Size Mode=::NewEnumerator3
    Uniform Sprite Size=&Multiply_Float(
     A=30
     B=&MaskFloatBySpawnGroup(
      Masked Float=&Multiply_Float(A=1, B=&MakeFloatFromVector(Vector=$User.ShoeScale_L, Channel=::NewEnumerator0))
      Passthrough Float=&Multiply_Float(A=1, B=&MakeFloatFromVector(Vector=$User.ShoeScale_R, Channel=::NewEnumerator0))
      Spawn Group Mask=$Particles.PartitionParticles.Partition
     )
    )
    UsePositionOffset=false
    Write Material Random=true
   )
   &AddVelocity(Rotation Coordinate Space=::World, Velocity=<0 0 100>)
   &Color(Color=&MultiplyLinearColor(Linear Color=$Particles.Color, Scale Factor=&MakeVector4_FromLinearColor(LinearColor=$System.TODColor)))
   &SubUVAnimation(
    Evaluation Type=::NewEnumerator0
    "Sprite Renderer "=DataInterface() {propertyValues {SpriteRenderer=&self:NiagaraEmitter_3.NiagaraSpriteRendererProperties_1}}
    SubUV Animation Mode:ENiagara_SubUVLookupModeV2=::NewEnumerator2
    UseEndFrame=false
    UseStartFrame=false
   )
  ]
  particleUpdateScript [
   &ParticleState()
   &SkeletalMeshLocation(
    Bone / Socket Index=&SelectIntFromArray(Direct Array Index=$Particles.PartitionParticles.Partition, Int Selection Array=DataInterface() {propertyValues {IntData [1,3]}})
    Bone Sampling Mode=::NewEnumerator4
    MaskSpawnGroup=false
    Sampled Position Offset=&Multiply_Vector(
     A=&AddVectorToPosition(
      Position=&RandomRangeVector(Minimum=<-5 0 -6>, Maximum=<5 0 6>)
      Vector=&MergeHelper_VectorSwitch(0=<-4 0 4>, 1=<-4 0 0>, 2=?, 3=?, Index=$Particles.PartitionParticles.Partition, Num=::NewEnumerator0)
     )
     B=$User.ShoeScale_L
    )
    Skeletal Mesh=$System.Skeletal Mesh Reader
   ) {module=SkeletalMeshLocation001}
   &ScaleSpriteSize(
    Uniform Curve Scale=&Multiply_Float(A=1.3, B=$System.ScaleShoes)
    Uniform Curve Sprite Scale=DataInterface() {propertyValues {Curve {keys [
        {interpMode=RCIM_Cubic, leaveTangent=2.2229, tangentMode=RCTM_Break, time=0, value=0.2}
        {arriveTangent=1, interpMode=RCIM_Cubic, leaveTangent=1, time=0.5, value=1.05}
        {arriveTangent=0.452, tangentMode=RCTM_User, time=1, value=1.2}
       ]}}}
   )
   &Drag(Drag=0.75, Use Linear Drag=true, Use Rotational Drag=true)
   &GravityForce(Gravity=<0 0 -500>)
   &SolveForcesAndVelocity(Clamp Velocity=false, Limit Acceleration=false)
   &DynamicMaterialParameters(
    Index 0 Param 1=&RandomRangeFloat(Minimum=0, Maximum=1)
    Index 0 Param 2=&FloatFromCurve(FloatCurve=DataInterface() {propertyValues {Curve {defaultValue=0, keys [{tangentMode=RCTM_User, time=0, value=0.35}, {tangentMode=RCTM_User, time=1, value=1}]}}})
    Index 0 Param 3=1
    Index 0 Param 4=&RandomRangeFloat(Minimum=1, Maximum=2)
    Module.Index0Float0Write=true
    Module.Index0Float1Write=true
    Module.Index0Float2Write=true
    Module.Index0Float3Write=true
   )
  ]
  rendererClasses [&NiagaraSpriteRendererProperties]
  renderers [{rendererClass=&NiagaraSpriteRendererProperties}]
 }
 Trail001 {
  emitterSpawnScript [&FN_RibbonIDOffsetUtility()]
  emitterUpdateScript [
   &EmitterState()
   &FN_SpawnAnimTrail(
    Number of Anim Trails:NiagaraInt32=&MakeCustomIntFromBool(
     Bool=&Boolean_OrOperation(Bool1=$User.IsSkydiving, Bool2=&Boolean_NotOperation(Bool=$User.isStage3))
     True INT=0
     False INT=2
    )
    Spawn Interval:NiagaraInt32=&MakeIntFromFloat(Float=&MakeCustomFloatFromBool(
      Bool=$User.IsSkydiving
      True Float=0
      False Float=&MakeCustomFloatFromBool(
       Bool=$User.HideFootFX
       True Float=0
       False Float=&MakeCustomFloatFromBool(Bool=$System.isRunning, True Float=12, False Float=0)
      )
     ))
    Spawn Mode:ENiagaraSkeletalSpawning=::NewEnumerator0
    Switch on ENiagara_AnimTrail_Source:ENiagara_AnimTrail_Source=::NewEnumerator1
   )
  ]
  particleSpawnScript [
   &FN_InitializeAnimTrailBoneIndecies(Second Bone or Socket Index=0)
   &FN_InitializeAnimTrailBoneIndecies(AnimTrail Selection=1, First Bone or Socket Index=2, Second Bone or Socket Index=2) {module=FN_InitializeAnimTrailBoneIndecies001}
   &FN_InitializeAnimTrail(
    AnimTrailWidth=&Multiply_Float(
     A=30
     B=&MaskFloatBySpawnGroup(
      Masked Float=&MakeFloatFromVector(Vector=$User.ShoeScale_L, Channel=::NewEnumerator0)
      Passthrough Float=&MakeFloatFromVector(Vector=$User.ShoeScale_R, Channel=::NewEnumerator0)
     )
    )
    Color=<rgba 1 1 1 0.5>
    Lifetime=0.15
    SkeletalMesh=$System.Skeletal Mesh Reader
   )
   &AddVelocity(Velocity=&FN_FrontEndVector3Modifier(Gameplay=<0 0 0>, Front End=<0 0 0>))
  ]
  particleUpdateScript [
   &SolveForcesAndVelocity(Clamp Velocity=false, Limit Acceleration=false)
   &FN_UpdateAnimTrail(SkeletalMesh=$System.Skeletal Mesh Reader) {module=FN_UpdateAnimTrail001}
   &ScaleColor(
    Linear Color Curve=DataInterface() {propertyValues {
      AlphaCurve {keys [<at 0.0144 1.2>,<at 0.0876 0.6>,<at 1 0>]}
      BlueCurve {keys [<at 0.2507 1>,<at 0.4595 0.8504>,<at 0.6162 0.5781>]}
      GreenCurve {keys [<at 0.2507 0.894>,<at 0.4595 0.8958>,<at 0.6162 0.3941>]}
      RedCurve {keys [<at 0.2507 0.6>,<at 0.4595 0.5375>,<at 0.6162 0.3469>]}
     }}
    Scale Mode=::NewEnumerator2
   )
   &DynamicMaterialParameters(
    Index 0 Param 1=&MergeHelper_FloatSwitch(0=0.5, 1=0, 2=?, 3=?, Index=&SpawnGroup(), Num=::NewEnumerator0)
    Index 0 Param 4=1
    Module.Index0Param0Write=true
    Module.Index0Param1Write=true
    Module.Index0Param2Write=true
    Module.Index0Param3Write=true
   ) {module=DynamicMaterialParameters001}
   &ScaleColor(Scale RGB=&MakeVectorFromLinearColorRGB(Color=$System.TODColor), ScaleA=true, ScaleRGB=true) {module=ScaleColor001}
   &ScaleRibbonWidth(Ribbon Width Scale=&FloatFromCurve(
     FloatCurve=DataInterface() {propertyValues {Curve {keys [<at 0 1 RCIM_Cubic>, {arriveTangent=-0.6839, tangentMode=RCTM_User, time=1, value=0.2943}]}}}
     Scale Curve=&FN_FrontEndFloatModifier(Gameplay=1, Front End=0.4)
    ))
  ]
  rendererClasses [&NiagaraRibbonRendererProperties]
  renderers [{rendererClass=&NiagaraRibbonRendererProperties}]
 }
 WaterDrops {
  emitterUpdateScript [
   &EmitterState(Life Cycle Mode=::NewEnumerator1, Loop Behavior=::NewEnumerator1, Loop Duration Mode=::NewEnumerator1, UseLoopDelay=false)
   &SpawnRate(
    SpawnRate=&MakeCustomFloatFromBool(Bool=$User.isStage3, True Float=10, False Float=0)
    Use Spawn Probability=true
   )
  ]
  particleSpawnScript [
   &PartitionParticles()
   &InitializeParticle(
    Color=<rgba 0.5435 0.5724 0.6094 9>
    Lifetime Mode=::NewEnumerator1
    Sprite Rotation Mode=::NewEnumerator2
    Sprite Size Mode=::NewEnumerator3
    Uniform Sprite Size=8
    UsePositionOffset=false
    Write Material Random=true
   )
   &CameraOffset(Camera Offset Amount=&MergeHelper_FloatSwitch(0=2, 1=2, 2=2, 3=2, 4=2, 5=2, 6=5, Index=$Particles.PartitionParticles.Partition, Num=::NewEnumerator5))
  ]
  particleUpdateScript [
   &ParticleState()
   &SkeletalMeshLocation(
    Bone / Socket Index=&MergeHelper_IntSwitch(
     0=&RandomRangeInt(Minimum=0, Maximum=9)
     1=&RandomRangeInt(Minimum=10, Maximum=19)
     2=?
     3=?
     4=?
     5=?
     Index=$Particles.PartitionParticles.Partition
     Num=::NewEnumerator0
    )
    "Interpolate Mesh Sampling (Update)"=true
    MaskSpawnGroup=false
    Mesh Sampling Type=::NewEnumerator5
    Orientation Sampling=::NewEnumerator1
    Sampled Position Offset=&MergeHelper_VectorSwitch(
     0=<0 0 0>
     1=<0 0 0>
     2=<0 0 0>
     3=<0 0 0>
     4=<0 0 0>
     5=<0 0 0>
     6=&RandomRangeVector(Minimum=<-2.5 -2.5 2>, Maximum=<2.5 2.5 2.5>)
     Index=$Particles.PartitionParticles.Partition
     Num=::NewEnumerator5
    )
    Sampled Position Offset Coordinate Space:ENiagaraSkelMeshTransforms=::NewEnumerator2
    Skeletal Mesh=$System.Skeletal Mesh Reader
    Socket Sampling Mode=::NewEnumerator5
    Use Sampled Mesh Attribute as Ribbon ID=false
   )
   &ScaleSpriteSize(
    Uniform Curve Scale=&Multiply_Float(
     A=&Multiply_Float(A=&FN_FrontEndFloatModifier(Gameplay=1.5, Front End=1), B=$System.ScaleShoes)
     B=&MaskFloatBySpawnGroup(
      Masked Float=&Multiply_Float(A=1, B=&MakeFloatFromVector(Vector=$User.ShoeScale_R, Channel=::NewEnumerator0))
      Passthrough Float=&Multiply_Float(A=1, B=&MakeFloatFromVector(Vector=$User.ShoeScale_L, Channel=::NewEnumerator0))
      Spawn Group Mask=$Particles.PartitionParticles.Partition
     )
    )
    Uniform Curve Sprite Scale=DataInterface() {propertyValues {Curve {keys [
        {arriveTangent=1.1867, interpMode=RCIM_Cubic, leaveTangent=1.1867, tangentMode=RCTM_User, time=0, value=0.4}
        {interpMode=RCIM_Cubic, leaveTangent=-1.3651, time=1, value=1}
       ]}}}
   ) {module=ScaleSpriteSize001}
   &DynamicMaterialParameters(
    Index 0 Param 1=1
    Index 0 Param 2=1
    Index 0 Param 3=&FloatFromCurve(FloatCurve=DataInterface() {propertyValues {Curve {keys [
         {arriveTangent=0.5904, interpMode=RCIM_Cubic, leaveTangent=0.5904, tangentMode=RCTM_User, time=0, value=0}
         {arriveTangent=2.1055, interpMode=RCIM_Cubic, leaveTangent=2.1055, tangentMode=RCTM_User, time=1, value=1}
        ]}}})
    Index 0 Param 4=&RandomRangeFloat(Minimum=0, Maximum=33)
    Index 1 Param 1=&FloatFromCurve(
     FloatCurve=DataInterface() {propertyValues {Curve {
        defaultValue=1
        keys [
         {tangentMode=RCTM_User, time=0, value=-1}
         {tangentMode=RCTM_User, time=0.2, value=-1}
         <at 0.2 -1>
         {arriveTangent=-0.0768, tangentMode=RCTM_User, time=0.5, value=-0.5}
         <at 0.6 0>
         <at 0.75 -0.1>
         <at 0.8 -0.8>
         <at 1 -1>
        ]
       }}}
     CurveIndex=Linked() {linkedVariable {name=NPC.FortniteNPC.TimeOfDay_0-1, type=NiagaraFloat}}
    )
    Index 1 Param 2=&FloatFromCurve(
     FloatCurve=DataInterface() {propertyValues {Curve {keys [
         <at 0.216 0.4 RCIM_Cubic>
         {interpMode=RCIM_Cubic, tangentMode=RCTM_User, time=0.3505, value=0.94}
         {interpMode=RCIM_Cubic, tangentMode=RCTM_User, time=0.71, value=0.94}
         <at 0.8425 0.4 RCIM_Cubic>
        ]}}}
     CurveIndex=Linked() {linkedVariable {name=NPC.FortniteNPC.TimeOfDay_0-1, type=NiagaraFloat}}
    )
    Module.Index0Param0Write=true
    Module.Index0Param1Write=true
    Module.Index0Param2Write=true
    Module.Index0Param3Write=true
    Module.Index1Param0Write=true
    Module.Index1Param1Write=true
    Module.Index1Param2Write=true
    Module.Index1Param3Write=true
    Write Parameter Index 1=true
   )
   /SetVariables(Particles.RimLightDynamic:Vector4f=&MakeVector4_FromLinearColor(LinearColor=&LerpLinearColors(
      StartColor=&MultiplyLinearColor(
       Linear Color=&MultiplyLinearColor(
        Linear Color=&ColorFromCurve(
         ColorCurve=DataInterface() {propertyValues {
           AlphaCurve {keys [<at 0 1>,<at 0.125 1>,<at 0.25 1>,<at 0.375 1>,<at 0.5 1>,<at 0.625 1>,<at 0.75 1>,<at 0.875 1>,<at 1 1>]}
           BlueCurve {keys [<at 0 0.01>,<at 0.041 0.01>,<at 0.083 0.01>,<at 0.125 0.01>,<at 0.166 0.0077>,<at 0.208 0.0045>,<at 0.25 0.0092>,<at 0.291 0.0136>,<at 0.333 0.03>,<at 0.375 0.04>,<at 0.416 0.0525>,<at 0.458 0.0525>,<at 0.5 0.0525>,<at 0.541 0.0525>,<at 0.583 0.0525>,<at 0.625 0.04>,<at 0.666 0.03>,<at 0.708 0.0152>,<at 0.75 0.0153>,<at 0.791 0.0079>,<at 0.833 0.0025>,<at 0.875 0.0031>,<at 0.916 0.01>,<at 0.958 0.01>,<at 1 0.01>]}
           GreenCurve {keys [<at 0 0.0068>,<at 0.041 0.0064>,<at 0.083 0.0064>,<at 0.125 0.0071>,<at 0.166 0.0056>,<at 0.208 0.0041>,<at 0.25 0.0103>,<at 0.291 0.0186>,<at 0.333 0.0156>,<at 0.375 0.0208>,<at 0.416 0.0272>,<at 0.458 0.0272>,<at 0.5 0.0272>,<at 0.541 0.0272>,<at 0.583 0.0272>,<at 0.625 0.0208>,<at 0.666 0.0156>,<at 0.708 0.0192>,<at 0.75 0.0153>,<at 0.791 0.0092>,<at 0.833 0.0051>,<at 0.875 0.0062>,<at 0.916 0.0064>,<at 0.958 0.0063>,<at 1 0.0068>]}
           RedCurve {keys [<at 0 0.0033>,<at 0.041 0.0029>,<at 0.083 0.0032>,<at 0.125 0.0036>,<at 0.166 0.01>,<at 0.208 0.01>,<at 0.25 0.02>,<at 0.291 0.03>,<at 0.333 0.0086>,<at 0.375 0.0115>,<at 0.416 0.0151>,<at 0.458 0.0151>,<at 0.5 0.0151>,<at 0.541 0.0151>,<at 0.583 0.0151>,<at 0.625 0.0115>,<at 0.666 0.0086>,<at 0.708 0.03>,<at 0.75 0.03>,<at 0.791 0.02>,<at 0.833 0.01>,<at 0.875 0.01>,<at 0.916 0.0022>,<at 0.958 0.0019>,<at 1 0.0033>]}
          }}
         CurveIndex=Linked() {linkedVariable {name=NPC.FortniteNPC.TimeOfDay_0-1, type=NiagaraFloat}}
        )
        Scale Factor=<0.5 0.5 0.5 1>
       )
       Scale Factor=&MakeFloat4FromVectorAndFloat(
        "Vector (RGB)"=&VectorFromFloat(Value=&FN_PickFloatDependOnPlatform(TargetPlatform=DataInterface() {propertyValues {Platforms {qualityLevelMask=-29}}}, TargetPlatformActive=0.2, TargetPlatformNoActive=1))
        "Float (Alpha)"=1
       )
      )
      EndColor=<rgba 0.0151 0.0272 0.0525 1>
      LerpFactor=&FN_FrontEndFloatModifier(Gameplay=0, Front End=1)
     ))) {isSetParameters=true}
   /SetVariables(Particles.HighlightDynamic:Vector4f=&MakeVector4_FromLinearColor(LinearColor=&LerpLinearColors(
      StartColor=&ColorFromCurve(
       ColorCurve=DataInterface() {propertyValues {
         AlphaCurve {keys [<at 0 1>,<at 0.125 1>,<at 0.25 1>,<at 0.375 1>,<at 0.5 1>,<at 0.625 1>,<at 0.75 1>,<at 0.875 1>,<at 1 1>]}
         BlueCurve {keys [<at 0 0.9216>,<at 0.041 0.8963>,<at 0.083 0.8308>,<at 0.125 0.7835>,<at 0.1865 0.2611>,<at 0.25 0.3013>,<at 0.333 0.3128>,<at 0.4094 0.2806>,<at 0.4889 0.2616>,<at 0.583 0.2814>,<at 0.666 0.2801>,<at 0.75 0.2471>,<at 0.791 0.2118>,<at 0.833 0.2338>,<at 0.875 0.2473>,<at 0.916 0.8388>,<at 0.958 0.9047>,<at 1 0.9216>]}
         GreenCurve {keys [<at 0 0.5616>,<at 0.041 0.5276>,<at 0.083 0.4795>,<at 0.125 0.5099>,<at 0.1865 0.2253>,<at 0.25 0.3543>,<at 0.333 0.4809>,<at 0.4094 0.4334>,<at 0.4889 0.389>,<at 0.583 0.2814>,<at 0.666 0.4785>,<at 0.75 0.2471>,<at 0.791 0.2686>,<at 0.833 0.4442>,<at 0.875 0.5188>,<at 0.916 0.5292>,<at 0.958 0.5688>,<at 1 0.6308>]}
         RedCurve {keys [<at 0 0.1584>,<at 0.041 0.1774>,<at 0.083 0.1644>,<at 0.125 0.1755>,<at 0.1865 0.6867>,<at 0.25 0.8148>,<at 0.333 0.8963>,<at 0.4094 0.9131>,<at 0.4889 0.9131>,<at 0.583 0.8714>,<at 0.666 0.8963>,<at 0.75 0.7529>,<at 0.791 0.7529>,<at 0.833 0.8469>,<at 0.875 0.8632>,<at 0.916 0.1791>,<at 0.958 0.1696>,<at 1 0.305>]}
        }}
       CurveIndex=Linked() {linkedVariable {name=NPC.FortniteNPC.TimeOfDay_0-1, type=NiagaraFloat}}
      )
      EndColor=<rgba 0.7 0.7 0.7 1>
      LerpFactor=&FN_FrontEndFloatModifier(Gameplay=1, Front End=1)
     ))) {isSetParameters=true}
   &AccelerationForce(Acceleration=&MergeHelper_VectorSwitch(
     0=<0 0 5>
     1=<0 0 5>
     2=<0 0 5>
     3=<0 0 5>
     4=<0 0 5>
     5=<0 0 5>
     6=<0 0 0>
     Index=$Particles.PartitionParticles.Partition
     Num=::NewEnumerator5
    ))
   &SolveForcesAndVelocity(Clamp Velocity=false, Limit Acceleration=false)
  ]
  rendererClasses [&NiagaraSpriteRendererProperties]
  renderers [{rendererClass=&NiagaraSpriteRendererProperties}]
 }
}
summary {
 systemName=NS_Shoe_WaveLaces_Idle
 emitters [
  {emitterName=SingleSplash, rendererClasses [&NiagaraSpriteRendererProperties]}
  {emitterName=DropsFishes, rendererClasses [&NiagaraSpriteRendererProperties]}
  {emitterName=GravityChunks, rendererClasses [&NiagaraSpriteRendererProperties]}
  {emitterName=WaterDrops, rendererClasses [&NiagaraSpriteRendererProperties]}
  {emitterName=Trail001, rendererClasses [&NiagaraRibbonRendererProperties]}
 ]
 userVariables [
  User.bHideFXOnClothes {type=NiagaraFloat, defaultValue=0, description=""}
  User.bisFootHidden {type=NiagaraBool}
  User.bIsFrontEnd {type=NiagaraBool}
  User.bIsFrontEndPreview {type=NiagaraBool}
  User.bisHandHidden {type=NiagaraBool}
  User.bisLit {type=NiagaraBool}
  User.bIsVariant {type=NiagaraFloat, defaultValue=0, description=""}
  User.EmissiveIndex {type=NiagaraInt32}
  User.EmissiveTint {type=LinearColor, description="", defaultValue {struct=LinearColor, value=<rgba 1 1 1 1>}}
  User.HideFootFX {type=NiagaraBool}
  User.isOffPawnPreview {type=NiagaraBool}
  User.IsSkydiving {type=NiagaraBool}
  User.isStage3 {type=NiagaraBool}
  User.RibbonIDOffset {type=NiagaraInt32}
  User.ShoeScale_L {type=Vector3f, description="", defaultValue {struct=Vector3f, value=<1 1 1>}}
  User.ShoeScale_R {type=Vector3f, description="", defaultValue {struct=Vector3f, value=<1 1 1>}}
 ]
}
